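# Material function for the Fruitcake player body that varies emissive with time of day, behind a static switch.
# A cosine of the TimeOfDay_0-1 collection parameter, shaped by power curves, forms a night mask.
# The mask scales incoming emissive toward an overall night brightness and fades in an emissive texture, weighted by strength and an unclamped camera-dot-normal power that favors camera-facing surfaces.
# Their sum replaces EmissiveColor in the material attributes.
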MaterialFunction /Game/Characters/Player/Female/Medium/Bodies/F_MED_Fruitcake/Materials/MF_Fruitcake_EmissiveTOD
graph {nodes [
  FunctionOutput(None=StaticSwitchParameter(True=NamedRerouteUsage(), False=NamedRerouteUsage()) {group=Fruitcake_EmissiveTOD, parameterName=UseFruitcake_EmissiveTOD, sortPriority=0})
  NamedRerouteDeclaration(None=FunctionInput() {inputType=FunctionInput_MaterialAttributes}) {name=MA, nodeColor=<rgba 1 0.0137 0 1>}
  n1 = GetMaterialAttributes(None=NamedRerouteUsage()) {attributeGetTypes [#1]}
  n2 = Multiply(A=Cosine(None=CollectionParameter() {collection=&FortniteMaterialParameters, parameterName=TimeOfDay_0-1})) {constB=0.5}
  n3 = OneMinus(None=Power(Base=Add(A=@n2) {constB=0.5}) {constExponent=8, desc=Day time EXP})
  n4 = Saturate(None=OneMinus(None=Power(Base=@n3) {desc=Night time EXP}))
  n5 = Multiply(
   A=@n1.EmissiveColor
   B=LinearInterpolate(B=ScalarParameter() {defaultValue=0.85, group=Fruitcake_EmissiveTOD, parameterName=OverallNightBrighness, sliderMax=1}, Alpha=@n4) {constA=1}
  )
  n6 = TextureSampleParameter2D() {group=Fruitcake_EmissiveTOD, parameterName=Emissive, sortPriority=1, texture=&T_BlacksRGB}
  n7 = ScalarParameter() {defaultValue=1, group=Fruitcake_EmissiveTOD, parameterName=EmissiveStrength, sliderMax=10, sortPriority=2}
  n8 = DotProduct(A=CameraVectorWS(), B=PixelNormalWS())
  n9 = ScalarParameter() {defaultValue=2, group=Fruitcake_EmissiveTOD, parameterName=EmissiveFresnelExp, sliderMax=8, sortPriority=2}
  n10 = Multiply(A=@n4, B=Multiply(A=Multiply(A=@n6.RGB, B=@n7), B=Power(Base=@n8, Exp=@n9)))
  NamedRerouteDeclaration(None=SetMaterialAttributes!(MaterialAttributes=@n1.MaterialAttributes, EmissiveColor=Add(A=@n5, B=@n10)) {attributeSetTypes [#1]}) {name=MA_Output, nodeColor=<rgba 0.7578 1 0 1>}
 ]}
